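# Niagara idle effect for the GoneFishing keytar cosmetic, layering fire, flickering lights, smoke and sparks around a shared center offset.
# Fire places burst mesh particles by array index, while sprite emitters add flame licks, rising sparks and flipbook smoke gated by owner speed.
# User.IsPickaxe blends fire placement and smoke motion for a pickaxe variant, and bIsFrontEndPreview boosts smoke alpha.
# A more complex movement-inheriting smoke emitter is present but disabled.
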NiagaraSystem /SparksCosmetics/Instrument/Keytar/GoneFishing/Effects/NS_Keytar_GoneFishing_Idle
dependencies {
 usedDataInterfaces [
  &NiagaraDataInterfaceCurve
  &NiagaraDataInterfaceVector2DCurve
  &NiagaraDataInterfaceSpriteRendererInfo
  &NiagaraDataInterfaceArrayFloat3
  &NiagaraDataInterfaceArrayFloat
  &NiagaraDataInterfaceColorCurve
 ]
 usedDynamicInputs [
  &ReturnNormalizedExecIndex
  &Vector2DFromCurve
  &RandomRangeVector2D
 ]
 usedModules [
  &self:SystemSpawnScript.SystemScriptSource.SystemScriptGraph.NiagaraNodeAssignment_1.SetVariables_#1
  &SystemState
 ]
}
emitters {
 Fire {
  emitterUpdateScript [
   &EmitterState()
   &SpawnBurst_Instantaneous(Spawn Count=8, Spawn Time=0.1, Use Spawn Probability=false, UseLoopCountLimit=true)
  ]
  particleSpawnScript [
   &InitializeParticle(
    Mesh Scale=&Multiply_VectorByFloat(
     Vector=<0.05 0.35 0.1>
     Float=&SelectFloatFromArray(
      Direct Array Index=&ReturnExecIndex()
      Float Selection Array=DataInterface() {propertyValues {FloatData [1,1,1,1,1,1,0.5,0.5]}}
     )
    )
    Mesh Scale Mode=::NewEnumerator4
    Position Offset=&Add_Vector(
     A=&Add_Vector(A=$System.CenterOffset, B=&Lerp_Vector(A=<0 0 -25>, B=<0 0 -24>, Alpha=$User.IsPickaxe))
     B=&SelectVectorFromArray(
      Direct Array Index=&ReturnExecIndex()
      Vector Selection Array=DataInterface() {propertyValues {FloatData [<0 0 7.1>, <0 0 14.2>, <0 0 21.3>, <0 0 28.4>, <0 0 35.5>, <0 0 42.6>, <-2 0.25 49.7>, <-2 0.25 0>]}}
     )
    )
    UsePositionOffset=true
    Write Material Random=true
   )
   &InitialMeshOrientation(
    Mesh Axis To Orient=&Lerp_Vector(A=<0 0 1>, B=<0 0 -1>, Alpha=$User.IsPickaxe)
    Mesh Orientation Mode=::NewEnumerator2
    Rotation=<0 0 0.2>
    Use Rotation Vector=true
   )
   &DynamicMaterialParameters(
    Index 0 Param 2=&RandomRangeFloat(Minimum=0, Maximum=1)
    Module.Index0Float0Write=true
    Module.Index0Float1Write=true
    Module.Index0Float2Write=true
    Module.Index0Float3Write=true
   )
  ]
  particleUpdateScript [
   &ParticleState(
    Kill Particles When Lifetime Has Elapsed=false
    Let Infinitely Lived Particles Die When Emitter Deactivates=true
    Loop Particles Lifetime=true
   )
   &ScaleColor(
    Scale Alpha=&FNTimeOfDayScalar(Night Scale=1, Morning Scale=1, Evening Scale=1)
    Scale RGB=&MakeVectorFromLinearColorRGB(Color=&RandomRangeLinearColor(
      Minimum=<rgba 1 0.2271 0.1757 1>
      Maximum=<rgba 0.9804 1 0.0087 1>
      Evaluation Type=::NewEnumerator1
      Randomness Mode=::NewEnumerator1
      Random Seed=&Add_Integer(A=&RandomRangeInt(Minimum=0, Maximum=500), B=$Emitter.LoopCount)
      Fixed Random Seed=true
     ))
    ScaleA=true
    ScaleRGB=true
   )
  ]
  rendererClasses [&NiagaraMeshRendererProperties]
  renderers [{rendererClass=&NiagaraMeshRendererProperties, rendererIndex=0}]
 }
 FlameLicks {
  emitterUpdateScript [
   &EmitterState()
   &SpawnRate(SpawnRate=8, Use Spawn Probability=false)
  ]
  particleSpawnScript [
   &InitializeParticle(
    Lifetime Max=0.6
    Lifetime Min=0.3
    Lifetime Mode=::NewEnumerator1
    Sprite Rotation Mode=::NewEnumerator2
    Sprite Size Mode=::NewEnumerator1
    Uniform Sprite Size Max=20
    Uniform Sprite Size Min=15
    UsePositionOffset=false
    Write Material Random=true
   )
   &DynamicMaterialParameters(
    Index 0 Param 2=&UniformRangedFloat(Minimum=0, Maximum=1)
    Index 0 Param 3=1
    Index 0 Param 4=&UniformRangedFloat(Minimum=3, Maximum=5)
    Module.Param0WriteEnabled=false
    Module.Param1WriteEnabled=true
    Module.Param2WriteEnabled=true
    Module.Param3WriteEnabled=true
   )
   &ShapeLocation(
    Box Size=<10 1 35>
    MaskSpawnGroup=false
    Offset=&Add_Vector(A=$System.CenterOffset, B=<0 2.5 0>)
    Offset Coordinate Space=::Local
    Offset Mode=::NewEnumerator0
    Override Seed=false
    Shape Primitive=::NewEnumerator2
    Surface Only=false
    UseOwnerScale=true
   ) {module=ShapeLocation001}
   &CameraOffset(Camera Offset Amount=1)
  ]
  particleUpdateScript [
   &ParticleState()
   &ShapeLocation(MaskSpawnGroup=false, Override Seed=false, Sphere Radius=15, Sphere Surface Distribution=1, UseOwnerScale=true) {enabled=false}
   &ScaleSpriteSize(Scale Factor=&Vector2DFromFloat(Value=&FloatFromCurve(FloatCurve=DataInterface() {propertyValues {Curve {keys [
          {arriveTangent=2.4034, interpMode=RCIM_Cubic, leaveTangent=2.4033, tangentMode=RCTM_User, time=0.0002, value=0.15}
          {arriveTangent=0.5342, interpMode=RCIM_Cubic, leaveTangent=0.5342, tangentMode=RCTM_User, time=0.3, value=0.575}
          {arriveTangent=-0.1019, interpMode=RCIM_Cubic, leaveTangent=-0.1019, tangentMode=RCTM_User, time=1.0041, value=0.7174}
         ]}}})))
   &SpriteRotationRate(Rotation Rate=50)
   &SolveForcesAndVelocity(Clamp Velocity=false, Limit Acceleration=false)
   /SetVariables(Particles.SubImageIndex=&FloatFromCurve(
     FloatCurve=DataInterface() {propertyValues {Curve {keys [{tangentMode=RCTM_User, time=0, value=0}, {tangentMode=RCTM_User, time=1, value=1}]}}}
     Scale Curve=40
    )) {isSetParameters=true}
   &DynamicMaterialParameters(
    Index 0 Param 1=&FloatFromCurve(FloatCurve=DataInterface() {propertyValues {Curve {keys [{arriveTangent=3.2297, interpMode=RCIM_Cubic, leaveTangent=3.2297, tangentMode=RCTM_User, time=0, value=0}, <at 0.75 1 RCIM_Cubic>]}}})
    Index 0 Param 2=2.5
    Index 0 Param 4=3
    Module.Param0WriteEnabled=true
    Module.Param1WriteEnabled=false
    Module.Param2WriteEnabled=false
    Module.Param3WriteEnabled=false
   ) {module=DynamicMaterialParameters001}
   &ScaleColor(
    Scale Alpha=0.25
    Scale RGB=&Multiply_VectorByFloat(
     Vector=&MakeVectorFromLinearColorRGB(Color=&ColorFromCurve(ColorCurve=DataInterface() {propertyValues {
         AlphaCurve {keys [<at 0 1 RCIM_Cubic>]}
         BlueCurve {keys [<at 0 0 RCIM_Cubic>,<at 0.1964 0>,<at 0.8699 0 RCIM_Cubic>]}
         GreenCurve {keys [<at 0 0.939 RCIM_Cubic>,<at 0.1964 0.2885>,<at 0.8699 0.0542 RCIM_Cubic>]}
         RedCurve {keys [<at 0 3 RCIM_Cubic>,<at 0.1964 2>,<at 0.8699 1 RCIM_Cubic>]}
        }}))
     Float=3
    )
    ScaleA=true
    ScaleRGB=true
   )
  ]
  rendererClasses [&NiagaraSpriteRendererProperties]
  renderers [{rendererClass=&NiagaraSpriteRendererProperties}]
 }
 Light {
  emitterUpdateScript [
   &EmitterState()
   &SpawnBurst_Instantaneous(Spawn Count=3, Use Spawn Probability=false, UseLoopCountLimit=true)
  ]
  particleSpawnScript [&InitializeParticle(
    Color=<rgba 25 4.125 0.5652 1>
    Lifetime=5
    Position Offset=&Add_Vector(
     A=$System.CenterOffset
     B=&SelectVectorFromArray(
      Direct Array Index=&ReturnExecIndex()
      Vector Selection Array=DataInterface() {propertyValues {FloatData [<0 3 18>, <0 3 -18>, <0 3 0>]}}
     )
    )
    Sprite Size Mode=::NewEnumerator3
    Uniform Sprite Size=5
    UsePositionOffset=true
    Write Material Random=true
   )]
  particleUpdateScript [
   &ParticleState(Kill Particles When Lifetime Has Elapsed=false)
   &InterpolateOverTime(
    Initial Float Value=0.5
    Rate Of Change=2
    Target Float Value:NiagaraFloat=&RandomRangeFloat(Minimum=0, Maximum=1, Evaluation Type=::NewEnumerator1)
   )
   &InterpolateOverTime(
    Rate Of Change=10
    Target Color Value:LinearColor=&RandomRangeLinearColor(Minimum=<rgba 0.5 0.0113 0.0113 1>, Maximum=<rgba 1.5 1.5 0.0548 1>, Evaluation Type=::NewEnumerator1)
    Variable Type=::NewEnumerator4
   ) {module=InterpolateOverTime001}
   &Light_Attributes(
    Color=1
    Light Position Offset:Vector3f=&Lerp_Vector(A=<0 0 -20>, B=<0 0 20>, Alpha=$Particles.InterpolateOverTime.InterpolateOverTime_Float.Moving Average)
    Light Radius:NiagaraFloat=&Lerp_Float(A=4, B=6, Alpha=$Particles.InterpolateOverTime.InterpolateOverTime_Float.Moving Average)
    Position=1
    Radius=1
    Scale RGB:Vector3f=&MakeVectorFromLinearColorRGB(Color=$Particles.InterpolateOverTime001.InterpolateOverTime_LinearColor.Moving Average)
    ScaleA=false
    ScaleRGB=true
    UsePositionOffset=true
   )
   &Light_Attributes(
    Color=1
    Light Position Offset:Vector3f=&Lerp_Vector(A=<0 0 -20>, B=<0 0 20>, Alpha=$Particles.InterpolateOverTime.InterpolateOverTime_Float.Moving Average)
    Light Radius:NiagaraFloat=&Lerp_Float(A=25, B=30, Alpha=$Particles.InterpolateOverTime.InterpolateOverTime_Float.Moving Average)
    Position=1
    Radius=1
    Scale RGB:Vector3f=&VectorFromFloat(Value=&Lerp_Float(A=0.5, B=2, Alpha=$Particles.InterpolateOverTime.InterpolateOverTime_Float.Moving Average))
    ScaleA=false
    ScaleRGB=true
    UsePositionOffset=true
   ) {enabled=false, module=Light_Attributes001}
  ]
  rendererClasses [&NiagaraLightRendererProperties, &NiagaraSpriteRendererProperties]
  renderers [
   {rendererClass=&NiagaraLightRendererProperties, rendererIndex=0}
   {rendererClass=&NiagaraSpriteRendererProperties, rendererIndex=1}
  ]
 }
 Smoke {
  emitterUpdateScript [
   &EmitterState()
   &SpawnRate(
    Spawn Probability=&MakeCustomFloatFromBool(
     Bool=&SetBoolByFloatComparison(A=&VectorLength(Vector=$Engine.Owner.Velocity), B=1000)
     True Float=0
     False Float=1
    )
    SpawnRate=14
    Use Spawn Probability=true
   )
  ]
  particleSpawnScript [
   &InitializeParticle(
    Color=<rgba 0.025 0.025 0.025 0.75>
    Lifetime Max=1.25
    Lifetime Min=0.75
    Lifetime Mode=::NewEnumerator1
    Position Offset=&Add_Vector(A=$System.CenterOffset, B=<0 3 0>)
    Position Offset Coordinate Space=::Local
    Sprite Rotation Mode=::NewEnumerator2
    Sprite Size Max=<18 18>
    Sprite Size Min=<12 12>
    Sprite Size Mode=::NewEnumerator2
    UsePositionOffset=true
    Write Material Random=true
   )
   &ShapeLocation(Box Size=<10 1 45>, MaskSpawnGroup=false, Override Seed=false, Shape Primitive=::NewEnumerator2, Surface Only=false, UseOwnerScale=true)
   &AddVelocity(Velocity=<0 15 0>, Velocity Speed Scale=&Lerp_Float(A=0, B=1, Alpha=$User.IsPickaxe)) {module=AddVelocity001}
  ]
  particleUpdateScript [
   &ParticleState()
   &SpriteRotationRate(Rotation Rate=&RandomRangeFloat(Minimum=-50, Maximum=50))
   &ScaleSpriteSize(
    Scale Sprite Size Mode=::NewEnumerator0
    Uniform Scale Factor=&FloatFromCurve(FloatCurve=DataInterface() {propertyValues {Curve {keys [<at 0 0.8>,<at 1 1.2>]}}}, Scale Curve=1.3)
   )
   &GravityForce(Gravity=&Multiply_VectorByFloat(
     Vector=&RandomRangeVector(Minimum=<0 0 25>, Maximum=<0 0 40>)
     Float=&Lerp_Float(A=0.5, B=2, Alpha=$User.IsPickaxe)
    ))
   &SubUVAnimation(
    End Frame Range Override=30
    "Sprite Renderer "=DataInterface() {propertyValues {SpriteRenderer=&self:Smoke_0.NiagaraSpriteRendererProperties_0}}
    Start Frame Range Override=&RandomRangeInt(Minimum=0, Maximum=5)
    UseEndFrame=true
    UseStartFrame=true
   )
   &ScaleColor(
    Scale Alpha=&FloatFromCurve(
     FloatCurve=DataInterface() {propertyValues {Curve {keys [
         {arriveTangent=7.7616, interpMode=RCIM_Cubic, leaveTangent=7.7616, tangentMode=RCTM_User, time=0, value=0}
         {arriveTangent=-0.0888, interpMode=RCIM_Cubic, leaveTangent=-0.0888, tangentMode=RCTM_User, time=0.1, value=1}
         {arriveTangent=-3.9148, interpMode=RCIM_Cubic, leaveTangent=-3.9148, tangentMode=RCTM_User, time=1, value=0}
        ]}}}
     Scale Curve=&MakeCustomFloatFromBool(Bool=$User.bIsFrontEndPreview, True Float=2, False Float=1)
    )
    ScaleA=true
    ScaleRGB=true
   )
   &Drag(Drag=&RandomRangeFloat(Minimum=2, Maximum=3.5), Use Linear Drag=true, Use Rotational Drag=true)
   &SolveForcesAndVelocity(Clamp Velocity=false, Limit Acceleration=false)
   &CameraOffset(Camera Offset Amount=2)
  ]
  rendererClasses [&NiagaraSpriteRendererProperties]
  renderers [{rendererClass=&NiagaraSpriteRendererProperties}]
 }
 SmokeComplex {
  bEnabled=false
  emitterUpdateScript [
   &EmitterState()
   &SpawnPerUnit(
    Max Movement Threshold=1000
    Module.Use Max Movement Threshold=true
    Module.Use Movement Tolerance=false
    Spawn Spacing=10
    Use Spawn Probability=false
   )
   &SpawnRate(
    Spawn Probability=&MakeCustomFloatFromBool(
     Bool=&SetBoolByFloatComparison(A=&VectorLength(Vector=$Engine.Owner.Velocity), B=1000)
     True Float=0
     False Float=1
    )
    SpawnRate=14
    Use Spawn Probability=true
   )
  ]
  particleSpawnScript [
   &InitializeParticle(
    Color=<rgba 0.025 0.025 0.025 0.75>
    Lifetime Max=1.25
    Lifetime Min=0.75
    Lifetime Mode=::NewEnumerator1
    Position Offset=&Add_Vector(A=$System.CenterOffset, B=<0 2 0>)
    Position Offset Coordinate Space=::Local
    Sprite Rotation Mode=::NewEnumerator2
    Sprite Size Max=<18 18>
    Sprite Size Min=<12 12>
    Sprite Size Mode=::NewEnumerator2
    UsePositionOffset=true
    Write Material Random=true
   )
   &ShapeLocation(Box Size=<10 1 45>, MaskSpawnGroup=false, Override Seed=false, Shape Primitive=::NewEnumerator2, Surface Only=false, UseOwnerScale=true)
   &AddVelocity(Velocity=<0 15 0>) {module=AddVelocity001}
   /SetVariables(Particles.InheritMovementAmount=1) {isSetParameters=true}
   /SetVariables(Particles.InitialPositionLocal:NiagaraPosition=&RotateVectorByQuaternion(
     Vector To Rotate=&Subtract_Position(A=$Particles.Position, B=$Engine.Owner.Position)
     Quaternion=/Niagara/DynamicInputs/Rotation/InvertQuaternion(Quat:Quat4f=$Engine.Owner.Rotation)
    )) {isSetParameters=true}
  ]
  particleUpdateScript [
   &ParticleState()
   &SpriteRotationRate(Rotation Rate=&RandomRangeFloat(Minimum=-50, Maximum=50))
   &ScaleSpriteSize(
    Scale Sprite Size Mode=::NewEnumerator0
    Uniform Scale Factor=&FloatFromCurve(FloatCurve=DataInterface() {propertyValues {Curve {keys [<at 0 0.8>,<at 1 1.2>]}}}, Scale Curve=1.3)
   )
   &GravityForce(Gravity=<0 0 80>)
   &SubUVAnimation(
    End Frame Range Override=30
    "Sprite Renderer "=DataInterface() {propertyValues {SpriteRenderer=&self:SmokeComplex_0.NiagaraSpriteRendererProperties_0}}
    Start Frame Range Override=&RandomRangeInt(Minimum=0, Maximum=5)
    UseEndFrame=true
    UseStartFrame=true
   )
   /SetVariables(
    Particles.InheritMovementAmount:NiagaraFloat=/Niagara/DynamicInputs/Math/Min_Float(
     A:NiagaraFloat=$Particles.InheritMovementAmount
     B:NiagaraFloat=&RemapRange(
      Input Value To Remap=&VectorLength(Vector=$Engine.Owner.Velocity)
      Input Min=1000
      Input Max=2000
      Output Min=1
      Output Max=0
     )
    )
    Transient.AlphaMult:NiagaraFloat=&NormalizeDistanceRange(
     Start Position:NiagaraPosition=$Particles.Initial.Position
     End Position:NiagaraPosition=$Particles.Position
     Distance:NiagaraFloat=100
     Invert Normalized Range:NiagaraBool=true
    )
   ) {isSetParameters=true}
   &ScaleColor(
    Scale Alpha=&Multiply_Float(
     A=&NormalizeDistanceRange(
      Start Position:NiagaraPosition=$Particles.Initial.Position
      End Position:NiagaraPosition=$Particles.Position
      Distance:NiagaraFloat=100
      Invert Normalized Range:NiagaraBool=true
     )
     B=&FloatFromCurve(
      FloatCurve=DataInterface() {propertyValues {Curve {keys [
          {arriveTangent=7.7616, interpMode=RCIM_Cubic, leaveTangent=7.7616, tangentMode=RCTM_User, time=0, value=0}
          {arriveTangent=-0.0888, interpMode=RCIM_Cubic, leaveTangent=-0.0888, tangentMode=RCTM_User, time=0.1, value=1}
          {arriveTangent=-3.9148, interpMode=RCIM_Cubic, leaveTangent=-3.9148, tangentMode=RCTM_User, time=1, value=0}
         ]}}}
      Scale Curve=&Multiply_Float(
       A=&MakeCustomFloatFromBool(Bool=$User.bIsFrontEndPreview, True Float=2, False Float=1)
       B=$Transient.AlphaMult
      )
     )
    )
    ScaleA=true
    ScaleRGB=true
   )
   &Drag(Drag=&RandomRangeFloat(Minimum=2, Maximum=3.5), Use Linear Drag=true, Use Rotational Drag=true)
   &InheritSourceMovement(
    Applied Position Scale=&VectorFromFloat(Value=&FloatFromCurve(
      FloatCurve=DataInterface() {propertyValues {Curve {keys [
          {arriveTangent=-1.8853, interpMode=RCIM_Cubic, leaveTangent=-1.8853, leaveTangentWeight=0.3333, tangentMode=RCTM_User, time=0, value=1}
          <at 1 0 RCIM_Cubic>
         ]}}}
      Scale Curve=&MakeCustomFloatFromBool(Bool=$User.bIsFrontEndPreview, True Float=0, False Float=$Particles.InheritMovementAmount)
     ))
    ApplyForce=false
    ApplyPosition=true
    ApplyPositionFalloff=false
   )
   &SolveForcesAndVelocity(Clamp Velocity=false, Limit Acceleration=false)
   &self:Advance_Age_by_Distance(DecayRate=0.04, MaxRadius=75) {enabled=false, module=Advance_Age_by_Distance}
   &KillParticles(Kill Particles=&SetBoolByFloatComparison(A=$Transient.AlphaMult, B=0, Comparison Type=::NewEnumerator7))
   &Color(Color=&ColorFromCurve(
     ColorCurve=DataInterface() {propertyValues {
       AlphaCurve {keys [<at 0 1 RCIM_Cubic>]}
       BlueCurve {keys [<at 0 0 RCIM_Cubic>,<at 0.025 1>,<at 0.975 0>,<at 1 1 RCIM_Cubic>]}
       GreenCurve {keys [<at 0 1 RCIM_Cubic>,<at 0.025 0>,<at 0.975 0>,<at 1 1 RCIM_Cubic>]}
       RedCurve {keys [<at 0 0 RCIM_Cubic>,<at 0.025 0>,<at 0.975 1>,<at 1 1 RCIM_Cubic>]}
      }}
     CurveIndex=&MakeFloatFromVector(Vector=$Transient.AdvectionVelocity, Channel=::NewEnumerator0)
    )) {enabled=false}
   /SetVariables(Transient.LocalPosition:Vector3f=&Add_Vector(
     A=&Subtract_Position(A=$Particles.Position, B=$Particles.Initial.Position)
     B=&RotateVectorByQuaternion(Vector To Rotate=Linked() {linkedVariable {name=Particles.InitialPositionLocal, type=NiagaraPosition}}, Quaternion=$Engine.Owner.Rotation)
    )) {isSetParameters=true}
   /SetVariables(Particles.PositionPicker:NiagaraPosition=&MakeCustomVectorFromBool(
     Bool=$User.bIsFrontEndPreview
     True Vector=&AddVectorToPosition(Position=$Engine.Owner.Position, Vector=$Transient.LocalPosition)
     False Vector=Linked() {linkedVariable {name=Particles.Position, type=NiagaraPosition}}
    )) {isSetParameters=true}
   /SetVariables(Particles.PositionPicker:NiagaraPosition=$Particles.Position) {enabled=false, isSetParameters=true}
  ]
  rendererClasses [&NiagaraSpriteRendererProperties]
  renderers [{rendererClass=&NiagaraSpriteRendererProperties}]
 }
 Sparks {
  emitterUpdateScript [
   &EmitterState()
   &SpawnRate(Spawn Probability=0.25, SpawnRate=30, Use Spawn Probability=true)
  ]
  particleSpawnScript [
   &InitializeParticle(
    Color Maximum=<rgba 2 1.2022 0.6482 1>
    Color Minimum=<rgba 1 0.2061 0.1339 1>
    Color Mode=::NewEnumerator2
    Lifetime=&UniformRangedFloat(Minimum=0.1, Maximum=0.85)
    Position Offset=$System.CenterOffset
    Position Offset Coordinate Space=::Local
    Sprite Size Mode=::NewEnumerator1
    Uniform Sprite Size Max=2
    Uniform Sprite Size Min=1
    UsePositionOffset=true
    Write Material Random=true
   )
   &AddVelocity(Velocity=&RandomRangeVector(Minimum=<0 50 0>, Maximum=<0 100 0>))
   &ShapeLocation(Box Size=<12 1 40>, MaskSpawnGroup=false, Override Seed=false, Shape Primitive=::NewEnumerator2, Surface Only=false, UseOwnerScale=true)
  ]
  particleUpdateScript [
   &ParticleState()
   /Niagara/Modules/Update/Forces/CurlNoiseForce(
    Noise Frequency=0.02
    Noise Strength=&FloatFromCurve(FloatCurve=DataInterface() {propertyValues {Curve {keys [<at 0 1>,<at 1 0>]}}}, Scale Curve=200)
    Pan Noise=true
    Pan Noise Field=<0 0 1>
   )
   &AccelerationForce(Acceleration=<0 0 250>)
   &Drag(Drag=5, Use Linear Drag=true, Use Rotational Drag=false)
   &ScaleColor(
    Scale Alpha=&FloatFromCurve(FloatCurve=DataInterface() {propertyValues {Curve {keys [
         {arriveTangent=11.9433, interpMode=RCIM_Cubic, leaveTangent=11.9433, tangentMode=RCTM_User, time=0, value=0}
         <at 0.2 1 RCIM_Cubic>
         <at 1 0 RCIM_Cubic>
        ]}}})
    Scale RGB=&MakeVectorFromLinearColorRGB(Color=&ColorFromCurve(ColorCurve=DataInterface() {propertyValues {
        AlphaCurve {keys [<at 0 1 RCIM_Cubic>]}
        BlueCurve {keys [<at 0 0.5069 RCIM_Cubic>,<at 1 0 RCIM_Cubic>]}
        GreenCurve {keys [<at 0 1.6422 RCIM_Cubic>,<at 1 0.06 RCIM_Cubic>]}
        RedCurve {keys [<at 0 4 RCIM_Cubic>,<at 1 1 RCIM_Cubic>]}
       }}))
    ScaleA=true
    ScaleRGB=true
   )
   &ScaleSpriteSize(Scale Factor=&Vector2DFromFloat(Value=&FloatFromCurve(FloatCurve=DataInterface() {propertyValues {Curve {keys [{arriveTangent=0.48, interpMode=RCIM_Cubic, time=0, value=1}, <at 1 0.5 RCIM_Cubic>]}}})))
   &SolveForcesAndVelocity(Clamp Velocity=false, Limit Acceleration=false)
   &ScaleSpriteSizeBySpeed(Max Scale Factor=<0.15 3>, Min Scale Factor=<0.25 1>)
  ]
  rendererClasses [&NiagaraSpriteRendererProperties]
  renderers [{rendererClass=&NiagaraSpriteRendererProperties}]
 }
}
summary {
 systemName=NS_Keytar_GoneFishing_Idle
 emitters [
  {emitterName=Smoke, rendererClasses [&NiagaraSpriteRendererProperties]}
  {emitterName=Fire, rendererClasses [&NiagaraMeshRendererProperties]}
  {emitterName=Light, rendererClasses [&NiagaraLightRendererProperties, &NiagaraSpriteRendererProperties]}
  {emitterName=FlameLicks, rendererClasses [&NiagaraSpriteRendererProperties]}
  {emitterName=Sparks, rendererClasses [&NiagaraSpriteRendererProperties]}
  {bEnabled=false, emitterName=SmokeComplex, rendererClasses [&NiagaraSpriteRendererProperties]}
 ]
 userVariables [
  User.bIsFrontEndPreview {type=NiagaraBool}
  User.IsPickaxe {type=NiagaraFloat, defaultValue=0, description=""}
 ]
}
