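# Physics asset for the Gone Fishing keytar cosmetic, giving its cookout props (hotdogs, burger patties, thighs and tongs) simple collision and constrained motion.
# Each bone, including the handle, has a single capsule body.
# Every prop is constrained directly to the handle in a star layout rather than a chain.
# Tight swing and twist limits, with some axes locked and tuned per prop, restrict how far each piece can move.
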PhysicsAsset /SparksCosmetics/Instrument/Keytar/GoneFishing/Meshes/GoneFishing_Keytar_Physics_FE
bodies {
 handle [{shapeType=Capsule, length=2.7948, radius=4.7244}]
 hotdog_01 [{shapeType=Capsule, length=8.7891, radius=1.4146, shapeName=hotdog_01_capsule, center=<0 1 0>, rotation=<rot -2.6947 88.2866 -80.0944>}]
 hotdog_02 [{shapeType=Capsule, length=7.8515, radius=1.4082, shapeName=hotdog_02_capsule, center=<0 1 0>, rotation=<rot 0 -89.3497 -83.6044>}]
 hotdog_03 [{shapeType=Capsule, length=5.5795, radius=1.4468, shapeName=hotdog_03_capsule, center=<0 1.2 0>, rotation=<rot 0 91.3301 -86.7026>}]
 patty [{shapeType=Capsule, length=0.1119, radius=3.2489, shapeName=patty_capsule, rotation=<rot 0 1.601 -51.1792>}]
 patty_with_cheese [{shapeType=Capsule, length=0.2806, radius=3.3952, shapeName=patty_with_cheese_capsule, rotation=<rot 0 -179.0337 -167.2047>}]
 thigh_01 [{shapeType=Capsule, length=4.7041, radius=1.9558, shapeName=thigh_01_capsule, center=<0 2 0>, rotation=<rot 0 -78.1927 -86.9136>}]
 thigh_02 [{shapeType=Capsule, length=4.6655, radius=1.8995, shapeName=thigh_02_capsule, center=<0 2 0>, rotation=<rot 0 -81.9468 -88.9393>}]
 tongs [{shapeType=Capsule, length=8.5062, radius=1.6673, shapeName=tongs_capsule, center=<2.1227 0.003 -2.6108>, rotation=<rot -0.013 -90 -40.7043>}]
}
constraints [
 {bone1Name=hotdog_01, bone2Name=handle, swing1LimitDegrees=45, swing1Motion=Locked, swing2LimitDegrees=2, twistLimitDegrees=25}
 {
  bone1Name=hotdog_02
  bone2Name=handle
  swing1LimitDegrees=45
  swing1Motion=Locked
  swing2LimitDegrees=4
  swing2Motion=Locked
  twistLimitDegrees=30
 }
 {bone1Name=hotdog_03, bone2Name=handle, swing1LimitDegrees=45, swing1Motion=Locked, swing2LimitDegrees=4, twistLimitDegrees=35}
 {bone1Name=patty, bone2Name=handle, swing1LimitDegrees=2, swing2LimitDegrees=2, twistLimitDegrees=6}
 {bone1Name=patty_with_cheese, bone2Name=handle, swing1LimitDegrees=2, swing2LimitDegrees=2, twistLimitDegrees=6}
 {bone1Name=thigh_01, bone2Name=handle, swing1LimitDegrees=0, swing2LimitDegrees=20, twistLimitDegrees=35}
 {bone1Name=thigh_02, bone2Name=handle, swing1LimitDegrees=0, swing1Motion=Locked, swing2LimitDegrees=20, twistLimitDegrees=35}
 {bone1Name=tongs, bone2Name=handle, swing1LimitDegrees=5, swing2LimitDegrees=5, twistLimitDegrees=45, twistMotion=Locked}
]
